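# M_ShellMaster is an Engine EditorShellMaterials lit master material producing tinted surfaces with an optional overlay.
# Static switches select UV or world-aligned triplanar sampling, blended by absolute world normal, for grayscale base and alpha-masked overlay.
# Base color darkens with camera distance, then optional banded inverse-Fresnel shading multiplies the result into ConvertFromDiffSpec.
# The world-size parameter names are swapped: "Overlay Texture Size" scales the base, "Diffuse Texture Size" the overlay.
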Material /Engine/EditorShellMaterials/M_ShellMaster
graph {
 properties {bUsedWithStaticLighting=true, pixelDepthOffsetMode=PDOM_Legacy, usedShadingModels=Default Lit}
 nodes [
  n1 = ScalarParameter() {defaultValue=100, parameterName=BandingFactor}
  n2 = Multiply(A=OneMinus(None=Fresnel()), B=@n1)
  n3 = Clamp(None=Divide(A=Ceil(None=@n2), B=@n1)) {minDefault=0.25}
  n4 = Constant() {r=1}
  n5 = Multiply(A=Constant() {r=0.5}, B=Divide(A=WorldPosition().XYZ, B=ScalarParameter() {defaultValue=32, parameterName=Overlay Texture Size}))
  n6 = ComponentMask(None=@n5) {b=true, g=true}
  n7 = ComponentMask(None=@n5) {b=true, r=true}
  n8 = Constant3Vector() {constant=<rgba 0 0 1 0>}
  n9 = Normalize(VectorInput=Abs(None=Transform(None=@n8)))
  n10 = ComponentMask(None=@n9) {g=true}
  n11 = ComponentMask(None=@n9) {b=true}
  n12 = LinearInterpolate(
   A=LinearInterpolate(
    A=TextureSampleParameter2D(UVs=@n6) {parameterName=BaseTexture, samplerType=SAMPLERTYPE_Grayscale, texture=&/T_Base}.RGB
    B=TextureSampleParameter2D(UVs=@n7) {parameterName=BaseTexture, samplerType=SAMPLERTYPE_Grayscale, texture=&T_Base}.RGB
    Alpha=@n10
   )
   B=TextureSampleParameter2D(UVs=ComponentMask(None=@n5) {g=true, r=true}) {parameterName=BaseTexture, samplerType=SAMPLERTYPE_Grayscale, texture=&T_Base}.RGB
   Alpha=@n11
  )
  n13 = VectorParameter() {parameterName=BaseTexCoords, defaultValue=<rgba 1 1 0 1>}
  n14 = TextureSampleParameter2D(UVs=Multiply(A=TextureCoordinate(), B=ComponentMask(None=@n13.RGB) {g=true, r=true})) {parameterName=BaseTexture, samplerType=SAMPLERTYPE_Grayscale, texture=&T_Base}
  n15 = VectorParameter() {parameterName=BaseColor, defaultValue=<rgba 0.5 0.5 0.5 1>}
  n16 = Constant() {r=0.0003}
  n17 = Distance(A=CameraPositionWS(), B=WorldPosition().XYZ)
  n18 = LinearInterpolate(
   A=Multiply(A=StaticSwitchParameter(True=@n12, False=@n14.RGB) {parameterName=World Aligned UVs}, B=@n15.RGB)
   B=Multiply(A=@n15.RGB, B=Constant() {r=0.75})
   Alpha=Clamp(None=Multiply(A=@n16, B=@n17)) {minDefault=0.5}
  )
  n19 = Multiply(A=Constant() {r=0.5}, B=Divide(A=WorldPosition().XYZ, B=ScalarParameter() {defaultValue=32, parameterName=Diffuse Texture Size}))
  n20 = LinearInterpolate(
   A=TextureSampleParameter2D(UVs=ComponentMask(None=@n19) {b=true, g=true}) {parameterName=OverlayTexture, texture=&/T_BaseOverlay}.A
   B=TextureSampleParameter2D(UVs=ComponentMask(None=@n19) {b=true, r=true}) {parameterName=OverlayTexture, texture=&T_BaseOverlay}.A
   Alpha=@n10
  )
  n21 = TextureSampleParameter2D(UVs=ComponentMask(None=@n19) {g=true, r=true}) {parameterName=OverlayTexture, texture=&T_BaseOverlay}
  n22 = Multiply(
   A=TextureCoordinate() {uTiling=0.5, vTiling=0.5}
   B=ComponentMask(None=VectorParameter() {parameterName=OverlayTexCoords, defaultValue=<rgba 1 1 0 1>}.RGB) {g=true, r=true}
  )
  n23 = LinearInterpolate(
   A=@n18
   B=VectorParameter() {parameterName=OverlayColor, defaultValue=<rgba 0.1 0.1 0.1 1>}.RGB
   Alpha=StaticSwitchParameter(
    True=LinearInterpolate(A=@n20, B=@n21.A, Alpha=@n11)
    False=TextureSampleParameter2D(UVs=@n22) {parameterName=OverlayTexture, texture=&T_BaseOverlay}.A
   ) {parameterName=World Aligned UVs}
  )
  MaterialFunctionCall(DiffuseColor=Multiply(
    A=StaticSwitchParameter(True=@n3, False=@n4) {defaultValue=true, parameterName=UseFresnel}
    B=StaticSwitchParameter(True=@n23, False=@n18) {parameterName=UseOverlay}
   )) {materialFunction=&ConvertFromDiffSpec}
 ]
}
